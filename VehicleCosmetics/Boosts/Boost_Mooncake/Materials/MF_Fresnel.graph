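# Material function that computes an adjustable Fresnel mask, stored with the Mooncake vehicle boost cosmetic's materials.
# It saturates the dot product of normal and camera vectors, defaulting to tangent space, and flips it to be edge-bright unless Invert Fresnel is on.
# A static switch shapes the result with either a Power exponent or, for Use Cheap contrast, an unclamped lerp between dark and bright inputs.
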MaterialFunction /VehicleCosmetics/Boosts/Boost_Mooncake/Materials/MF_Fresnel
graph {nodes [
  n1 = FunctionInput() {bUsePreviewValueAsDefault=true, inputName=Cheap contrast dark, inputType=FunctionInput_Scalar, sortPriority=11, previewValue=<0 0 1 1>}
  n2 = FunctionInput() {bUsePreviewValueAsDefault=true, inputName=Cheap contrast bright, inputType=FunctionInput_Scalar, sortPriority=12, previewValue=<1 0 1 1>}
  n3 = Transform(None=CameraVectorWS()) {transformSourceType=TRANSFORMSOURCE_World, transformType=TRANSFORM_Tangent}
  MaterialExpressionConstantClamp_9 = Saturate(None=DotProduct(
    A=FunctionInput() {bUsePreviewValueAsDefault=true, inputName=Normal Vector, previewValue=<0 0 1 1>}
    B=FunctionInput(Preview=@n3) {bUsePreviewValueAsDefault=true, inputName=Camera Vector, sortPriority=1}
   ))
  n4 = FunctionInput(Preview=StaticBool()) {bUsePreviewValueAsDefault=true, inputName=Invert Fresnel, inputType=FunctionInput_StaticBool, sortPriority=4, previewValue=<0 0 1 1>}
  n5 = StaticSwitch(True=@MaterialExpressionConstantClamp_9, False=OneMinus(None=@MaterialExpressionConstantClamp_9), Value=@n4)
  n6 = FunctionInput() {bUsePreviewValueAsDefault=true, inputName=Power, inputType=FunctionInput_Scalar, sortPriority=9, previewValue=<1 0 1 1>}
  n7 = FunctionInput(Preview=StaticBool()) {bUsePreviewValueAsDefault=true, inputName=Use Cheap contrast, inputType=FunctionInput_StaticBool, sortPriority=10, previewValue=<0 0 1 1>}
  FunctionOutput(None=StaticSwitch(True=LinearInterpolate(A=@n1, B=@n2, Alpha=@n5), False=Power(Base=@n5, Exp=@n6), Value=@n7))
 ]}
